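# Material function for weapon skin materials that applies SplineThicken to masked parts of an incoming material.
# SplineThicken, fed base and tip widths plus the input normal, adds to the existing World Position Offset, while a flattenable tube normal map replaces the normal.
# A selectable channel of the Mask input blends between original and thickened attributes.
# A minimum-width and root/tip taper network is present but unconnected to the output.
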MaterialFunction /Game/Weapons/WeaponSkins/MaterialFunction/FN_CharacterSplineThicken
graph {nodes [
  n1 = FunctionInput() {inputName=Material, inputType=FunctionInput_MaterialAttributes}
  n2 = GetMaterialAttributes(None=@n1) {attributeGetTypes [#1, #2]}
  n3 = MaterialFunctionCall(
   Normal=TextureSample() {samplerType=SAMPLERTYPE_Normal, texture=/Game/Animation/Props/Kite/Textures/Tube_Normal_Tube_N}.RGB
   Flatness=ScalarParameter() {group=SplineThicken, parameterName=ST_NormalFlatness, sortPriority=33}
  ) {materialFunction=&FlattenNormal}
  n4 = Add(
   A=@n2.WorldPositionOffset
   B=MaterialFunctionCall(
    WidthBase=ScalarParameter() {defaultValue=2, group=SplineThicken, parameterName=ST_WidthBase}
    WidthTip=ScalarParameter() {defaultValue=0.001, group=SplineThicken, parameterName=ST_WidthTip}
    AngleCorrectedNormals=StaticBoolParameter() {group=SplineThicken, parameterName=ST_AngleCorrectedNormals}
    AdditionalNormal=@n2.Normal
   ) {materialFunction=&SplineThicken}.WorldPositionOffset
  )
  n5 = FunctionInput(Preview=Constant4Vector() {constant=<rgba 1 1 1 1>}.RGBA) {inputName=Mask, inputType=FunctionInput_Vector4}
  FunctionOutput(None=BlendMaterialAttributes(
    A=GetMaterialAttributes(None=@n1).MaterialAttributes
    B=SetMaterialAttributes!(MaterialAttributes=@n2.MaterialAttributes, Normal=@n3.Result, WorldPositionOffset=@n4) {attributeSetTypes [#2, #1]}
    Alpha=ChannelMaskParameter(None=@n5) {group=SplineThicken, parameterName=ST_SplineThickenMaskChannel}.R
   ))
  MaterialFunctionCall()
  n6 = MaterialFunctionCall!(
   WorldWidth=ScalarParameter() {defaultValue=0.02, group=Width, parameterName=MinWorldWidth}
   PixelWidth=ScalarParameter() {defaultValue=0.2, group=Width, parameterName=MinPixelWidth}
  ) {materialFunction=&/MF_MinWidth}
  n7 = OneMinus(None=MaterialFunctionCall(UV Channel=TextureCoordinate()) {materialFunction=&LinearGradient}.VGradient)
  n8 = ScalarParameter() {defaultValue=0.01, group=Width, parameterName=MaxRoot}
  n9 = LinearInterpolate(
   A=@n6.Result
   B=Multiply(A=@n6.Result, B=ScalarParameter() {defaultValue=0.9, group=Width, parameterName=RootMult})
   Alpha=OneMinus(None=MaterialFunctionCall(Alpha=@n7, Max=@n8) {materialFunction=&SmoothStep}.Result)
  )
  n10 = ScalarParameter() {defaultValue=0.25, group=Width, parameterName=TipMult}
  n11 = ScalarParameter() {defaultValue=0.99, group=Width, parameterName=MinTip}
  Multiply(
   A=LinearInterpolate(A=ScalarParameter() {defaultValue=0.5, parameterName=widthMinMult}, B=ScalarParameter() {defaultValue=1.5, parameterName=widthMaxMult})
   B=LinearInterpolate(
    A=@n9
    B=Multiply(A=@n9, B=@n10)
    Alpha=MaterialFunctionCall(Alpha=@n7, Min=@n11) {materialFunction=&SmoothStep}.Result
   )
  )
  Power(Base=@n7)
 ]}
