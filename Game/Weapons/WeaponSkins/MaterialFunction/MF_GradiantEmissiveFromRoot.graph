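# Material function for weapon skins that overrides the emissive color of incoming material attributes with a two-color gradient.
# An offset second-UV lookup into a radial gradient texture is scaled, power-shaped and saturated to blend Color A into Color B.
# A shading path switch clamps the result only on mobile.
# Fresnel, object-radius and pre-skinned-position sphere mask branches are present but do not feed the output.
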MaterialFunction /Game/Weapons/WeaponSkins/MaterialFunction/MF_GradiantEmissiveFromRoot
graph {nodes [
  Constant()
  n1 = GetMaterialAttributes(None=FunctionInput() {inputType=FunctionInput_MaterialAttributes}) {attributeGetTypes [#1]}
  n2 = TextureCoordinate() {coordinateIndex=1}
  n3 = AppendVector(A=ScalarParameter() {parameterName=OffsetX}, B=ScalarParameter() {defaultValue=-0.381, parameterName=OffsetY})
  n4 = Multiply(
   A=TextureSample(UVs=Add(A=@n2, B=@n3)) {texture=/Game/Athena/HUD/Materials/T_UI_RadialGradient}.R
   B=ScalarParameter() {defaultValue=678.4948, parameterName=CenterColoamount}
  ) {constB=1.2}
  n5 = ScalarParameter() {defaultValue=0.6629, parameterName=TransitonPower}
  n6 = LinearInterpolate(
   A=VectorParameter() {parameterName=Color A, defaultValue=<rgba 0 0.5606 1 1>}.RGB
   B=VectorParameter() {parameterName=Color B, defaultValue=<rgba 1 0 0.0512 1>}.RGB
   Alpha=Saturate(None=Power(Base=@n4, Exp=@n5) {constExponent=3})
  )
  n7 = Saturate(None=@n6)
  FunctionOutput(None=SetMaterialAttributes!(
    MaterialAttributes=@n1.MaterialAttributes
    EmissiveColor=ShadingPathSwitch(Default=@n6, Deferred=@n6, Mobile=@n7) {desc=Migrated from FeatureLevelSwitch}
   ) {attributeSetTypes [#2]})
  n8 = ScalarParameter() {defaultValue=0.02, parameterName=FresnelEMissiveEX}
  n9 = Transform(None=@n1.Normal)
  LinearInterpolate(A=@n6, Alpha=Saturate(None=Fresnel(ExponentIn=@n8, Normal=@n9) {baseReflectFraction=0, exponent=0.1})) {constAlpha=0, constB=0}
  Multiply()
  n10 = ScalarParameter() {defaultValue=500, parameterName=DivideRadius}
  Saturate(None=Divide(A=ObjectRadius(), B=@n10))
  SphereMask(
   A=VertexInterpolator(VS=PreSkinnedPosition()).PS
   B=VectorParameter() {parameterName=Centerpoint, defaultValue=<rgba 0 0 0 1>}.RGB
   Radius=@n10
   Hardness=ScalarParameter() {defaultValue=0.33, parameterName=Hardness}
  )
 ]}
