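# Weapon skin material function that adds an audio-reactive scrolling emissive to incoming material attributes.
# UV or pre-skinned sphere-mask coordinates tile and pan, forming smoothstep and sphere-mask bands multiplied by random noise.
# Bands plus spectrum blue hue-shift a base color, scaled by a multiplier that is clamped on mobile.
# Spectrum red blends this over existing emissive; the amplitude lerp has identical endpoints, so amplitude has no effect.
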MaterialFunction /Game/Weapons/WeaponSkins/MaterialFunction/MF_ScrollingEmissive
graph {nodes [
  n1 = GetMaterialAttributes(None=FunctionInput() {inputType=FunctionInput_MaterialAttributes}) {attributeGetTypes [#1]}
  n2 = PreSkinnedPosition()
  n3 = StaticSwitchParameter(
   True=TextureCoordinate() {vTiling=3}
   False=SphereMask(
    A=Constant3Vector()
    B=VertexInterpolator(VS=@n2).PS
    Radius=ScalarParameter() {defaultValue=100, group=ScrollingEmissive, parameterName=MainGradientRadius}
    Hardness=ScalarParameter() {group=ScrollingEmissive, parameterName=MainGradientHardness}
   ) {attenuationRadius=0.5, hardnessPercent=0}
  ) {defaultValue=true, group=ScrollingEmissive, parameterName=UseUVCoords}
  n4 = AppendVector(A=Constant() {r=0.1}, B=ScalarParameter() {defaultValue=3, group=ScrollingEmissive, parameterName=Tile})
  n5 = Constant() {r=0.33}
  n6 = Constant()
  n7 = Panner(
   Coordinate=Add(A=Multiply(A=@n3, B=@n4), B=AppendVector(A=@n5, B=@n6))
   Speed=AppendVector(A=Constant(), B=ScalarParameter() {defaultValue=2, group=ScrollingEmissive, parameterName=Speed})
  ) {speedY=1}
  n8 = Frac(None=@n7)
  n9 = MaterialFunctionCall(
   Alpha=ComponentMask(None=@n8) {g=true}
   Min=ScalarParameter() {defaultValue=0.4, group=ScrollingEmissive, parameterName=SmoothstepMin}
   Max=ScalarParameter() {defaultValue=0.5, group=ScrollingEmissive, parameterName=SmoothstepMax}
  ) {materialFunction=&SmoothStep}
  n10 = SphereMask(
   A=@n8
   B=Constant() {r=0.5}
   Radius=ScalarParameter() {defaultValue=0.39, group=ScrollingEmissive, parameterName=SecondaryGradientRadius}
   Hardness=ScalarParameter() {group=ScrollingEmissive, parameterName=SecondaryGradientHardness}
  ) {attenuationRadius=0.5, hardnessPercent=0}
  n11 = ScalarParameter() {group=ScrollingEmissive, parameterName=NoiseDensityBias}
  n12 = CollectionParameter() {collection=&FortniteMaterialParameters, parameterName=EventAudio_Spectrum}
  n13 = MaterialFunctionCall(Float4=@n12) {materialFunction=&BreakOutFloat4Components}
  n14 = Add(
   A=Multiply(
    A=Multiply(A=@n9.Result, B=@n10)
    B=MaterialFunctionCall!(PrimaryTextureCoordinates=@n7, 2ndOrderTextureCoordinates, RandomSeed, NoiseDensityBias=@n11, Bypass Noise Influence Crosstalk) {materialFunction=/Game/Effects/Fort_Effects/Materials/_Functions/MF_RandomUVSample}.Noise
   )
   B=@n13.B
  )
  n15 = ScalarParameter() {defaultValue=10, group=ScrollingEmissive, parameterName=EmoteEmissiveMult}
  n16 = Multiply(
   A=MaterialFunctionCall(
    Hue Shift Percentage=Frac(None=@n14)
    Texture=VectorParameter() {group=ScrollingEmissive, parameterName=EmoteBaseColor, defaultValue=<rgba 3 0 0 0>}.RGB
   ) {materialFunction=&HueShift}.Result
   B=ShadingPathSwitch(Default=@n15, Deferred=@n15, Mobile=Saturate(None=@n15)) {desc=Migrated from FeatureLevelSwitch}
  )
  n17 = CollectionParameter() {collection=&FortniteMaterialParameters, parameterName=EventAudio_Amplitude}
  n18 = Power(Base=Saturate(None=LinearInterpolate(A=@n13.R, B=@n13.R, Alpha=@n17)), Exp=Constant() {r=0.25}) {constExponent=0.25}
  FunctionOutput(None=SetMaterialAttributes!(MaterialAttributes=@n1.MaterialAttributes, EmissiveColor=LinearInterpolate(A=@n1.EmissiveColor, B=@n16, Alpha=@n18)) {attributeSetTypes [#1]})
  n19 = CollectionParameter() {
   collection=&FortniteMaterialParameters
   desc=Deprecated
   parameterName=Emote_AudioLow
  }
  Reroute(None=@n19)
 ]}
