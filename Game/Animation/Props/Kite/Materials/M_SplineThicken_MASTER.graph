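# Master material for kite prop animation assets that applies character spline thickening on top of the shared character material functions.
# The PreFX base feeds the spline-thicken function with a constant full mask, then passes through static switches for root-gradient emissive, scrolling emissive and emissive flicker.
# The PostFX function finishes it, taking the PreFX AO for its AO, distance blend and Fresnel inputs.
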Material /Game/Animation/Props/Kite/Materials/M_SplineThicken_MASTER
graph {
 properties {
  bTangentSpaceNormal=false
  bUseMaterialAttributes=true
  bUsedWithInstancedStaticMeshes=true
  bUsedWithMeshDeformer=true
  bUsedWithNanite=true
  bUsedWithSkeletalMesh=true
  pixelDepthOffsetMode=PDOM_Legacy
  usedShadingModels=Default Lit
 }
 nodes [
  n1 = MaterialFunctionCall() {materialFunction=&BaseFN_CharacterFunction_PreFX}
  n2 = MaterialFunctionCall!(Mask=Constant() {r=1}, Material=@n1.Result) {materialFunction=/Game/Weapons/WeaponSkins/MaterialFunction/FN_CharacterSplineThicken}
  n3 = StaticSwitchParameter(
   True=MaterialFunctionCall!(In=@n2.Result) {materialFunction=/Game/Weapons/WeaponSkins/MaterialFunction/MF_GradiantEmissiveFromRoot}.Result
   False=@n2.Result
  ) {parameterName=Gradiate Emissive From Root}
  n4 = StaticSwitchParameter(True=MaterialFunctionCall!(In=@n3) {materialFunction=/Game/Weapons/WeaponSkins/MaterialFunction/MF_ScrollingEmissive}.Result, False=@n3) {parameterName=Scrolling Emissive}
  MaterialFunctionCall(
   AO=@n1.AO
   Distance Blend=@n1.AO
   Fresnel=@n1.AO
   Material=StaticSwitchParameter(
    True=MaterialFunctionCall!(Material=@n4, PerPixel Motion Offset) {materialFunction=/Game/Global/Base4Layers_2022/MaterialFunctions/MF_EmissiveFlicker}.Result
    False=@n4
   ) {group=Emissive, parameterName=EmissiveFlicker}
  ) {materialFunction=&BaseFN_CharacterFunction_PostFX}
 ]
}
